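; Athena reactive prop Blueprint that bursts a particle effect when damaged and can swap its static mesh.
; It binds to its damage event on BeginPlay and ignores damage caused by the global environment ability actors.
; FireEffect, optionally gated by DoOnce, spawns a scaled emitter at the hit location or a particle transform source, then applies any replacement mesh.
; It also returns marker display data and position offset.
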(event EventBeginPlay
  (EventDispatchers|BindEventtoOnDamaged (delegate Custom|OnDamagedCustom)))

(event Custom|OnDamagedCustom
  (params (DamagedActor :type Actor%20Object%20Reference) (Damage :type Float%20%28single-precision%29) (InstigatedBy :type Controller%20Object%20Reference) (DamageCauser :type Actor%20Object%20Reference) (HitLocation :type Vector) (FHitComponent :type Primitive%20Component%20Object%20Reference) (BoneName :type Name) (Momentum :type Vector))
  (if (not (or (Utilities|Equal(Class) (Utilities|GetClass DamageCauser) "/Script/FortniteGame.FortGlobalEnvironmentAbilityActor") (Utilities|Equal(Class) (Utilities|GetClass DamageCauser) "/Script/FortniteGame.FortGlobalEnvironmentAbilityActorAthena")))
    (CallFunction|FireEffect HitLocation)))

(event Custom|FireEffect
  (params (HitLocation :type Vector))
  (bind _meshtoreplaceondamage (Variables|Default|GetMeshToReplaceOnDamage))
  (if (Variables|Default|GetOnlyOnce)
    (Utilities|FlowControl|DoOnce false
      (:Completed
        _)))
  (Effects|SpawnEmitteratLocation (Variables|Default|GetEmitterOnDamage) (Math|Vector|SelectVector HitLocation (Transformation|GetWorldLocation (Variables|Default|GetParticleTransformSource)) (Variables|Default|GetSpawnParticlesFromHitLocation)) (Transformation|GetWorldRotation (Variables|Default|GetParticleTransformSource)) (Math|Vector|MakeVector (Variables|Default|GetEmitter%20Scale%20X) (Variables|Default|GetEmitter%20Scale%20Y) (Variables|Default|GetEmitter%20Scale%20Z)) true "AutoRelease")
  (Utilities|IsValid _meshtoreplaceondamage
    (:Is%20Valid
      (Variables|Mesh|SetStaticMesh _meshtoreplaceondamage))))

(fn GetMarkerDisplayData
  (return (Variables|Marking|GetMarkerDisplay)))

(fn GetMarkerPositionOffset
  (return (Variables|Marking|GetMarkerPositionOffset)))

(fn ConstructionScript)
