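; Destructible Athena reactive prop that catches fire below a health threshold and explodes when destroyed by non-pickaxe damage, disabling itself if its playlist setting is off.
; A ShouldDie override holds death at zero health; pickaxe kills die immediately, others drop resource generation and explode after a random delay.
; Replicated bShouldDie plays the explosion cue and runs a one-shot server sphere overlap that applies explosion effects, Curie fire and launches to pawns and vehicles.
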(event EventBeginPlay
  (|CallParentFunction)
  (if (Networking|HasAuthority)
    (Class|FortAsyncActionCurrentPlaylistReady|PlaylistDataReadyAsync
      (:OnReady
        (if (not (ScalableFloat|GetValueasBool (Variables|Default|GetIsEnabled)))
          (Gameplay|SilentDie true))))))

(event Collision|EventActorBeginOverlap
  (params (OtherActor :type Actor%20Object%20Reference)))

(event EventTick
  (params (DeltaSeconds :type Float%20%28single-precision%29)))

(event Custom|DeathFX
  ;[1 1 1] PickAxe Death Fx
  (Effects|Components|ParticleSystem|SpawnFXAtLocation :FXSystemAsset (CallFunction|GetDeathParticles) :Location (Transformation|GetActorLocation) :Scale "1.000000,1.000000,1.000000" :bAutoDestroy true :PoolingMethod "AutoRelease" :bAutoActivate true)
  (Projectile|RemoveGameplayCue%28Looping%29Local self (Variables|Default|GetFireCue) (Ability|GameplayCue|MakeGameplayCueParameters 0.0)))

(event EventFireEffect
  (params (HitLocation :type Vector :default "0, 0, 0"))
  ;[1 1 1] Damage State Notify
  (|CallParentFunction HitLocation))

(event Custom|ExplodeAndDie
  ;[1 1 1] If Destroyed by Anything but PickAxe, Explode
  (Rendering|SetActorHiddenInGame true)
  (Variables|Default|SetbShouldDie true))

(event EventOnDamageServer
  (params (Damage :type Float%20%28single-precision%29) (DamageTags :type Gameplay%20Tag%20Container%20Structure%20%28by%20ref%29) (Momentum :type Vector :default "0, 0, 0") (HitInfo :type Hit%20Result%20Structure%20%28by%20ref%29) (InstigatedBy :type Controller%20Object%20Reference) (DamageCauser :type Actor%20Object%20Reference) (EffectContext :type Gameplay%20Effect%20Context%20Handle%20Structure))
  ;[1 1 1] If Destroyed by Anything but PickAxe, Explode
  (bind _asfort_player_controller (Utilities|Casting|CastToFortPlayerController InstigatedBy)
    (:then
      (Variables|Default|SetLastDamageInstigator _asfort_player_controller))
    (:CastFailed
      _))
  (Utilities|FlowControl|Sequence
    (:then_0
      (if (Variables|Bools|GetHealthIsZero)
        (Utilities|FlowControl|DoOnce false
          (:Completed
            ;;[1 1 1] Check If Destroyed by Pick Axe
            (bind _returnvalue (MissionHelpers|HasTags DamageTags (Variables|GameplayTags|GetTC_MeleeDamage) true false))
            (if _returnvalue
              (CallFunction|DeathFX)
              (Variables|Default|SetDestroyedByPickAxe true)
              (Variables|Default|SetbShouldDie true)
              (else
                (Variables|Default|SetCanGenerateResources false)
                (Utilities|Time|SetTimerbyEvent (delegate Custom|ExplodeAndDie) (Math|Random|RandomFloatinRange 1.7 2.0))))))))
    (:then_1
      ;;[1 1 1] Set On Fire
      (if (<= (Health|GetHealthPercent) (Variables|Default|GetOnFireThreshold))
        (Utilities|FlowControl|DoOnce false
          (:Completed
            (Utilities|IsValid (Variables|Default|GetLastDamageInstigator)
              (:Is%20Valid
                (GameplayEffects|FortApplyGameplayEffectToTarget :self (Ability|GetFortAbilitySystemComponent (Pawn|GetControlledPawn (Variables|Default|GetLastDamageInstigator))) :GameplayEffectClass "/Game/Athena/Items/EnvironmentalItems/ReactiveProps/GE_Athena_ReactiveProps_OnFire.GE_Athena_ReactiveProps_OnFire_C" :TargetOfEffect (Ability|GetFortAbilitySystemComponent self) :OptionalSourceObject (Pawn|GetControlledPawn (Variables|Default|GetLastDamageInstigator)) :Level 0.0))
              (:Is%20Not%20Valid
                (Environment|ApplyGlobalEnvironmentGameplayEffecttoActor self "/Game/Athena/Items/EnvironmentalItems/ReactiveProps/GE_Athena_ReactiveProps_OnFire.GE_Athena_ReactiveProps_OnFire_C")))
            (CallFunction|OnFireFx)))))))

(event Custom|ExplosionTrace
  (bind _lastdamageinstigator (Variables|Default|GetLastDamageInstigator))
  (Utilities|FlowControl|DoOnce false
    (:Completed
      (bind _outactors (Collision|SphereOverlapActors (Transformation|GetActorLocation) (ScalableFloat|GetValueatLevel (Variables|Default|GetRadius)) (Utilities|Array|MakeArray "ObjectTypeQuery3" "ObjectTypeQuery1" "ObjectTypeQuery2" "ObjectTypeQuery5")))
      (for _array_element _outactors
        ;[1 0.9863 0.1497 1 12] Only damage other actors of this class if they haven't already started to explode to prevent excess work from being done.
        (bind _asathena_prop_burst_reaction_exploding (Utilities|Casting|CastToAthena_Prop_BurstReaction_Exploding _array_element)
          (:then
            (if (Variables|Bools|GetHealthIsZero _asathena_prop_burst_reaction_exploding)
              _
              (else
                _)))
          (:CastFailed
            _))
        (Variables|Default|SetCurrentPlayerToLaunch _array_element)
        (Utilities|FlowControl|Sequence
          (:then_0
            (CallFunction|ExecuteFireApplicationinRadiusifCurieEnabled _array_element))
          (:then_1
            ;[1 1 1] Apply Damage
            (Utilities|IsValid _lastdamageinstigator
              (:Is%20Valid
                (GameplayEffects|FortApplyGameplayEffectToTarget :self (Ability|GetFortAbilitySystemComponent (Pawn|GetControlledPawn _lastdamageinstigator)) :GameplayEffectClass (Variables|GameplayEffects|GetGE_Explode) :TargetOfEffect (Ability|GetFortAbilitySystemComponent _array_element) :OptionalSourceObject self :OptionalEffectCauser self :OptionalDamageSource self :Level 0.0))
              (:Is%20Not%20Valid
                (bind _asfort_pawn (Utilities|Casting|CastToFortPawn _array_element)
                  (:then
                    (Utilities|IsinVehicle _asfort_pawn
                      (:false
                        _)))
                  (:CastFailed
                    (Utilities|Casting|CastToFortVehicleInterface _array_element
                      (:then
                        (Environment|ApplyGlobalEnvironmentDamagetoActor _array_element (ScalableFloat|GetValueatLevel (Variables|Default|GetVehicleDamage))))
                      (:CastFailed
                        _))))
                (Environment|ApplyGlobalEnvironmentGameplayEffecttoActor _array_element (Variables|GameplayEffects|GetGE_Explode)))))
          (:then_2
            ;[1 1 1] Launch Player
            (bind _asfort_pawn_2 (Utilities|Casting|CastToFortPawn _array_element)
              (:then
                (if (not (or (GameplayTags|HasMatchingGameplayTag _asfort_pawn_2 "(TagName=\"Gameplay.Status.KnockbackImmunity\")") (GameplayTags|HasMatchingGameplayTag _asfort_pawn_2 "(TagName=\"Gameplay.Action.Player.AthenaInVehicle\")")))
                  (Character|LaunchCharacter _asfort_pawn_2 (Utilities|CalculateLaunchVelocity (Transformation|GetActorLocation _array_element) (Transformation|GetActorLocation)) true true)))
              (:CastFailed
                ;;[1 1 1] LaunchVehicle
                (bind _asfort_vehicle_interface (Utilities|Casting|CastToFortVehicleInterface _array_element))
                (bind _returnvalue (Vehicle|GetVehicleActor _asfort_vehicle_interface :!class /Script/FortniteGame.FortVehicleInterface))
                (Audio|BroadcastAbilityImpactatLocation 0 _returnvalue (select (>= (Math|Vector|VectorLength (Transformation|GetVelocity _returnvalue)) 5e+02) (Math|Vector|MakeVector (.x (* (Utilities|ClamptoMinAngleBetween (Math|Vector|Normalize (- (+ (Transformation|GetActorLocation _returnvalue) (Transformation|GetVelocity _returnvalue)) (Transformation|GetActorLocation _returnvalue))) (Math|Vector|ProjectVectorontoPlane (Math|Vector|Normalize (- (+ (Transformation|GetActorLocation _returnvalue) (Transformation|GetVelocity _returnvalue)) (Transformation|GetActorLocation _returnvalue))) "0, 0,1.000000") 3e+01 false) (ScalableFloat|GetValueatLevel (Variables|Default|GetVehicleImpulseStrength)))) (.y (* (Utilities|ClamptoMinAngleBetween (Math|Vector|Normalize (- (+ (Transformation|GetActorLocation _returnvalue) (Transformation|GetVelocity _returnvalue)) (Transformation|GetActorLocation _returnvalue))) (Math|Vector|ProjectVectorontoPlane (Math|Vector|Normalize (- (+ (Transformation|GetActorLocation _returnvalue) (Transformation|GetVelocity _returnvalue)) (Transformation|GetActorLocation _returnvalue))) "0, 0,1.000000") 3e+01 false) (ScalableFloat|GetValueatLevel (Variables|Default|GetVehicleImpulseStrength)))) (ScalableFloat|GetValueatLevel (Variables|Default|GetVehicleImpulseStrengthZ))) (Math|Vector|MakeVector (.x (* (Utilities|ClamptoMinAngleBetween (Math|Vector|Normalize (- (Transformation|GetActorLocation _array_element) (Transformation|GetActorLocation))) (Math|Vector|ProjectVectorontoPlane (Math|Vector|Normalize (- (Transformation|GetActorLocation _array_element) (Transformation|GetActorLocation))) "0, 0,1.000000") 3e+01 false) (ScalableFloat|GetValueatLevel (Variables|Default|GetVehicleImpulseStrength)))) (.y (* (Utilities|ClamptoMinAngleBetween (Math|Vector|Normalize (- (Transformation|GetActorLocation _array_element) (Transformation|GetActorLocation))) (Math|Vector|ProjectVectorontoPlane (Math|Vector|Normalize (- (Transformation|GetActorLocation _array_element) (Transformation|GetActorLocation))) "0, 0,1.000000") 3e+01 false) (ScalableFloat|GetValueatLevel (Variables|Default|GetVehicleImpulseStrength)))) (ScalableFloat|GetValueatLevel (Variables|Default|GetVehicleImpulseStrengthZ)))) 777.0)))))))))

(event EventOnDeathServer
  (params (Damage :type Float%20%28single-precision%29) (DamageTags :type Gameplay%20Tag%20Container%20Structure%20%28by%20ref%29) (Momentum :type Vector :default "0, 0, 0") (HitInfo :type Hit%20Result%20Structure%20%28by%20ref%29) (InstigatedBy :type Controller%20Object%20Reference) (DamageCauser :type Actor%20Object%20Reference) (EffectContext :type Gameplay%20Effect%20Context%20Handle%20Structure))
  ;[1 1 1] If Destroyed by Anything but PickAxe, Explode
  (bind _returnvalue (MissionHelpers|HasTags DamageTags (GameplayTags|MakeGameplayTagContainerfromArray (Utilities|Array|MakeArray "(TagName=\"Gameplay.Effect.InstantDeath.Silent\")" "(TagName=\"Gameplay.Effect.InstantDeath.StructuralSupport\")" "(TagName=\"Gameplay.Effect.InstantDeath.PersistentUpdate\")" "(TagName=\"Gameplay.Effect.InstantDeath.Environment\")" "(TagName=\"Gameplay.Effect.InstantDeath.Detachment\")" "(TagName=\"Gameplay.Effect.InstantDeath.Despawn\")" "(TagName=\"Gameplay.Effect.InstantDeath.Culled\")")) false false))
  (if _returnvalue
    _
    (elif (Variables|Default|GetDestroyedByPickAxe)
      _
      (else
        (CallFunction|ExplosionTrace)
        (Variables|Default|SetbShouldDie true)))))

(event Custom|OnFireFx
  ;[1 1 1] If Destroyed by Anything but PickAxe, Explode
  ;;[1 1 1] Set On Fire
  (Projectile|AddGameplayCue%28Looping%29Local self (Variables|Default|GetFireCue) (Ability|GameplayCue|MakeGameplayCueParameters 0.0)))

(event EventOnDeathPlayEffects
  (params (Damage :type Float%20%28single-precision%29) (DamageTags :type Gameplay%20Tag%20Container%20Structure%20%28by%20ref%29) (Momentum :type Vector :default "0, 0, 0") (HitInfo :type Hit%20Result%20Structure%20%28by%20ref%29) (InstigatedBy :type Fort%20Pawn%20Object%20Reference) (DamageCauser :type Actor%20Object%20Reference) (EffectContext :type Gameplay%20Effect%20Context%20Handle%20Structure))
  ;[1 1 1] If Destroyed by Anything but PickAxe, Explode
  (CallFunction|StopPreExplodeAudio))

(event Custom|GrowTimeline
  ;[1 1 1] Pre Explode Grow Timeline
  (if (not (<= (Health|GetHealthPercent) 0))
    (Utilities|FlowControl|DoOnce false
      (:Completed
        (bind _preexplodesound (Variables|Default|GetPreExplodeSound)
          (:then
            (bind _returnvalue (Audio|SpawnSoundatLocation _preexplodesound (Transformation|GetActorLocation)))
            (Variables|Default|SetPreExplodeAudio _returnvalue))
          (:else))))))

(event Custom|DeactivateEmitter
  ;[1 1 1] Deactivate attached spawned emitters on death (to avoid having them linger for a second, they destroy automatically)
  (if (not (Networking|IsServer))
    (Projectile|RemoveGameplayCue%28Looping%29Local self (Variables|Default|GetFireCue) (Ability|GameplayCue|MakeGameplayCueParameters 0.0))
    (for _array_element (Variables|DamageBasedEmitters|GetSpawnedDamageStateEmitters)
      (Components|Activation|Deactivate _array_element))))

(fn CanAttemptGenerateResources
  (params (InTags :type Gameplay%20Tag%20Container%20Structure%20%28by%20ref%29) (InstigatorController :type Controller%20Object%20Reference))
  (return (Variables|Default|GetCanGenerateResources)))

(fn ExecuteFireApplicationInRadiusIfCurieEnabled
  (params (Fire%20Application%20Target :type Actor%20Object%20Reference))
  (Networking|SwitchHasAuthority
    (:Authority
      (if (Curie|IsCurieActive)
        (Utilities|Casting|CastToLandscape Fire%20Application%20Target
          (:then
            (LocalVariables|SetTargetedLandscape true))
          (:CastFailed
            _))
        (Curie|ExecuteFireApplication (Variables|Default|GetLastDamageInstigator) self Fire%20Application%20Target 1e+02 (Collision|MakeHitResult false false 0.0 0.0 (Transformation|GetActorLocation (select (LocalVariables|GetTargetedLandscape) self Fire%20Application%20Target)) (Transformation|GetActorLocation (select (LocalVariables|GetTargetedLandscape) self Fire%20Application%20Target)) "0,0,1" "0,0,1" 0 Fire%20Application%20Target) -1 (ScalableFloat|GetValueatLevel (Variables|Default|GetRadius)))))))

(fn OnRep_bShouldDie
  (if (not (Variables|Default|GetDestroyedByPickAxe))
    (Projectile|RemoveGameplayCue%28Looping%29Local self (Variables|Default|GetFireCue) (Ability|GameplayCue|MakeGameplayCueParameters 0.0))
    (GameplayCue|ExecuteGameplayCue%28Burst%29Local self (Variables|GameplayCues|GetExplosionCUE) (Ability|GameplayCue|MakeGameplayCueParameters 0.0) false)
    (if (Networking|IsServer)
      (CallFunction|ExplosionTrace))))

(fn OnRep_DamageStateNotify
  ;[1 1 1] Use Current Health to figure out which emitters to spawn
  (bind (_array_element _array_index) (Utilities|Array|ForEachLoopwithBreak (Variables|DamageBasedEmitters|GetDamageStateParticleSystems))
    (:LoopBody
      (if (.Spawned_19_ADE1ECD74E521455DA9695B1FDB47581 (Utilities|Struct|BreakAthenaDamageStateParticles _array_element))
        _
        (elif (<= (Health|GetHealthPercent) (.HealthPercent_24_AC35B253419518DA53906DB5CCAD79F8 (Utilities|Struct|BreakAthenaDamageStateParticles _array_element)))
          ;[1 1 1] Set that this element has spawned (to avoid doing it twice)
          (Utilities|Array|SetArrayElem (Variables|DamageBasedEmitters|GetDamageStateParticleSystems) _array_index (Utilities|Struct|MakeAthenaDamageStateParticles (Utilities|Struct|BreakAthenaDamageStateParticles _array_element) true (.LocalSpawnPosition_9_507DFDCB4A99E944783EEAB44C495149 (Utilities|Struct|BreakAthenaDamageStateParticles _array_element)) (.LocalSpawnRotation_12_86D59DAF4DC65EF2E0CEB2BD7740F557 (Utilities|Struct|BreakAthenaDamageStateParticles _array_element)) (.LocalSpawnScale_20_D001987D4D7EB90DF56AAC9B5F8133A8 (Utilities|Struct|BreakAthenaDamageStateParticles _array_element))))
          ;[1 1 1] Spawn the effect
          (bind _returnvalue (AddComponent|NODE_AddParticleSystemComponent-0 :RelativeTransform_Location (.LocalSpawnPosition_9_507DFDCB4A99E944783EEAB44C495149 (Utilities|Struct|BreakAthenaDamageStateParticles _array_element)) :RelativeTransform_Rotation (.LocalSpawnRotation_12_86D59DAF4DC65EF2E0CEB2BD7740F557 (Utilities|Struct|BreakAthenaDamageStateParticles _array_element)) :RelativeTransform_Scale (.LocalSpawnScale_20_D001987D4D7EB90DF56AAC9B5F8133A8 (Utilities|Struct|BreakAthenaDamageStateParticles _array_element))))
          (Effects|Components|ParticleSystem|SetTemplate _returnvalue (Utilities|Struct|BreakAthenaDamageStateParticles _array_element))
          (Utilities|Array|Add (Variables|DamageBasedEmitters|GetSpawnedDamageStateEmitters) _returnvalue))))))

(fn ShouldDie
  (params (Damage :type Float%20%28single-precision%29) (EventInstigator :type Controller%20Object%20Reference) (DamageCauser :type Actor%20Object%20Reference) (bIsTest :type Boolean) (EffectContext :type Gameplay%20Effect%20Context%20Handle%20Structure%20%28by%20ref%29))
  ;[1 1 1] This Function get's called on Death. Health Is Zero get's set to true so that the damage event triggers the Death Fx. Should die is set to false so that the actors collision stays intact
  (Variables|Bools|SetHealthIsZero true)
  (return (Variables|Default|GetbShouldDie)))

(fn StopPreExplodeAudio
  (bind _preexplodeaudio (Variables|Default|GetPreExplodeAudio)
    (:then
      (if (Audio|Components|Audio|IsPlaying _preexplodeaudio)
        (Audio|Components|Audio|FadeOut _preexplodeaudio 0.25 0.0 "Linear" :!class /Script/Engine.AudioComponent)))
    (:else)))

(fn ConstructionScript
  (|CallParentFunction))
